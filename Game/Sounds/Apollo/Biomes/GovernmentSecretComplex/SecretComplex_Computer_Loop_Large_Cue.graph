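# Looping ambient cue for a large computer in the Government Secret Complex biome.
# A mixer favors a looping machine hum with a half-second fade-in over a repeating layer of randomly chosen, briefly delayed, pitch-modulated computer one-shots.
# Each layer has its own attenuation asset.
# The first two quality levels share this mix, the other two are empty, and overridden concurrency limits simultaneous instances.
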SoundCue /Game/Sounds/Apollo/Biomes/GovernmentSecretComplex/SecretComplex_Computer_Loop_Large_Cue
root=QualityLevel_0
nodes [
 {
  id=Attenuation_0
  children [Enveloper_0]
  properties {attenuationSettings=&/SecretComplex_Computer_Large_Att}
 }
 {
  id=Attenuation_1
  children [Looping_0]
  properties {attenuationSettings=&/SecretComplex_Computer_Att}
 }
 {id=Delay_0, children [Modulator_1], properties {delayMax=1, delayMin=0.5}}
 {
  id=Enveloper_0
  children [Modulator_6]
  properties {pitchCurve {editorCurveData {keys [{time=0}]}}, volumeCurve {editorCurveData {keys [<at 0 0>,<at 0.5 1>]}}}
 }
 {id=Looping_0, children [Delay_0]}
 {id=Mixer_1, children [Attenuation_1, Attenuation_0], properties {inputVolume [0.5,1.5]}}
 {id=Modulator_1, children [Random_0], properties {pitchMax=1.15, volumeMax=1, volumeMin=1}}
 {id=Modulator_6, children [WavePlayer_0], properties {volumeMax=1, volumeMin=1}}
 {id=QualityLevel_0, children [Mixer_1,Mixer_1,null,null]}
 {id=Random_0, children [WavePlayer_5,WavePlayer_6,WavePlayer_7,WavePlayer_8], properties {weights [1,1,1,1]}}
 {
  id=WavePlayer_0
  properties {bLooping=true, soundWaveAssetPtr=/Game/Athena/Sounds/Props/MountainBaseComputer/ComputerTransimtter_Running_Loop_Large_01}
 }
 {id=WavePlayer_5, properties {soundWaveAssetPtr=&/SecretComplex_Computer_Oneshot_02}}
 {id=WavePlayer_6, properties {soundWaveAssetPtr=&/SecretComplex_Computer_Oneshot_03}}
 {id=WavePlayer_7, properties {soundWaveAssetPtr=&/SecretComplex_Computer_Oneshot_04}}
 {id=WavePlayer_8, properties {soundWaveAssetPtr=&/SecretComplex_Computer_Oneshot_05}}
]
properties {
 bOverrideConcurrency=true
 maxDistance=2250
 soundClassObject=&Fort_Gadgets
 volumeMultiplier=1
 attenuationOverrides {
  bEnableFocusInterpolation=true
  falloffDistance=600
  nonFocusVolumeAttenuation=0.3
  occlusionVolumeAttenuation=0.3
  stereoSpread=0
  attenuationShapeExtents {x=200}
 }
 concurrencyOverrides {maxCount=3, resolutionRule=StopLowestPriority}
}
