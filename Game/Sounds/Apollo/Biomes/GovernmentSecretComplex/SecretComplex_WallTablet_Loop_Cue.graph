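# Looping ambient sound cue for a wall tablet in the Government Secret Complex biome, classed as a gadget sound.
# A quality-level node routes the first two levels to one looping wave player and leaves the other two empty, so those levels are silent.
# It plays quietly and pitched up, uses the complex's computer attenuation asset, and overrides concurrency to allow one instance.
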SoundCue /Game/Sounds/Apollo/Biomes/GovernmentSecretComplex/SecretComplex_WallTablet_Loop_Cue
root=QualityLevel_0
nodes [
 {id=QualityLevel_0, children [WavePlayer_2,WavePlayer_2,null,null]}
 {
  id=WavePlayer_2
  properties {bLooping=true, soundWaveAssetPtr=&/SecretComplex_WallTab_Loop_02}
 }
]
properties {
 attenuationSettings=&/SecretComplex_Computer_Att
 bOverrideConcurrency=true
 maxDistance=1200
 pitchMultiplier=1.3
 soundClassObject=&Fort_Gadgets
 volumeMultiplier=0.25
 attenuationOverrides {
  bEnableFocusInterpolation=true
  falloffDistance=600
  nonFocusVolumeAttenuation=0.3
  occlusionVolumeAttenuation=0.3
  stereoSpread=0
  attenuationShapeExtents {x=200}
 }
 concurrencyOverrides {maxCount=1}
}
